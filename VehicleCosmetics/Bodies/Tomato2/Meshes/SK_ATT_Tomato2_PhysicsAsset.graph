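# Physics asset for the SK_ATT_Tomato2 skeletal mesh, part of the Tomato2 body in the vehicle cosmetics content.
# It defines no constraints between bodies, and the file shows no body setups or collision shapes.
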PhysicsAsset /VehicleCosmetics/Bodies/Tomato2/Meshes/SK_ATT_Tomato2_PhysicsAsset
constraints []
